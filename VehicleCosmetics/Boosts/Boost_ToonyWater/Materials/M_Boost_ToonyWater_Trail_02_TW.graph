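# Translucent unlit Niagara ribbon material for a foamy trail in the ToonyWater vehicle boost.
# Soft cloud noise warps two panning water normal layers, which distort a UV-gradient mask shaping the trail.
# Panning clouds and water detail, pushed to high contrast, erode that mask into foam, and a noisy center band is subtracted from it.
# Particle color tints the emissive, while particle alpha and a U-gradient fade drive the separately weighted foam opacity.
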Material /VehicleCosmetics/Boosts/Boost_ToonyWater/Materials/M_Boost_ToonyWater_Trail_02_TW
graph {
 properties {
  bCastRayTracedShadows=false
  bUsedWithNiagaraRibbons=true
  blendMode=BLEND_Translucent
  pixelDepthOffsetMode=PDOM_Legacy
  refractionMethod=RM_PixelNormalOffset
  shadingModel=MSM_Unlit
  usedShadingModels=Unlit
 }
 nodes [
  n1 = ParticleColor()
  n2 = TextureSample(UVs=Panner(Coordinate=TextureCoordinate() {uTiling=0.25, vTiling=0.2}) {speedX=-0.125, speedY=0.01}) {texture=/DealershipSharedLibrary/FX/Textures/Noise/T_VCVFX_CloudsSoft_RGB}
  n3 = AppendVector(A=@n2.R, B=@n2.R)
  n4 = Multiply(A=Constant() {r=0.25}, B=@n3)
  n5 = Add(A=TextureCoordinate() {uTiling=1.875, vTiling=0.25}, B=@n4)
  n6 = Multiply(A=Time(), B=Constant() {desc="original = 2.5", r=2})
  n7 = Add(A=@n4, B=TextureCoordinate() {uTiling=2.25, vTiling=0.25})
  n8 = Add(
   A=TextureSample(UVs=Panner(Coordinate=@n5, Time=@n6) {speedX=-1, speedY=-0.05}) {samplerType=SAMPLERTYPE_Normal, texture=/DealershipSharedLibrary/FX/Textures/Noise/T_VCVFX_Water02_N}.RGB
   B=TextureSample(UVs=Panner(Coordinate=@n7, Time=@n6) {speedX=-1, speedY=0.02}) {samplerType=SAMPLERTYPE_Normal, texture=&T_VCVFX_Water02_N}.RGB
  )
  n9 = Add(A=Add(A=Constant2Vector() {g=0.2}, B=TextureCoordinate()), B=Multiply(A=Constant() {r=0.5}, B=ComponentMask(None=@n8) {g=true, r=true}))
  n10 = ComponentMask(None=@n9) {g=true}
  n11 = OneMinus(None=@n10)
  n12 = Multiply(A=Saturate(None=ConstantBiasScale(None=@n10) {bias=-0.285, scale=1}), B=Saturate(None=ConstantBiasScale(None=@n11) {bias=0.2, scale=1}))
  n13 = ComponentMask(None=@n9) {r=true}
  n14 = Saturate(None=ConstantBiasScale(None=OneMinus(None=@n13)) {bias=0.15, scale=1})
  n15 = ConstantBiasScale(None=Multiply(A=Multiply(A=@n12, B=@n14), B=Constant() {r=5})) {bias=-0.15, scale=8}
  n16 = Multiply(A=@n3, B=ScalarParameter() {defaultValue=1, parameterName=UVDistort})
  n17 = Panner(Coordinate=Add(A=TextureCoordinate() {uTiling=1.25, vTiling=0.5}, B=@n16)) {speedX=-1.25, speedY=-0.05}
  n18 = Panner(Coordinate=Add(A=TextureCoordinate() {uTiling=1.5, vTiling=0.5}, B=@n16)) {speedX=-1.25, speedY=0.02}
  n19 = Desaturation(None=Add(
    A=TextureSample(UVs=@n17) {texture=/VehicleCosmetics/Boosts/Boost_ToonyWater/Textures/T_VCVFX_Boost_ToonyWater_Clouds2_RGB}.RGB
    B=TextureSample(UVs=@n18) {texture=&T_VCVFX_Boost_ToonyWater_Clouds2_RGB}.RGB
   )) {luminanceFactors=<rgba 0.3 0.59 0.11 1>}
  n20 = Desaturation(None=ConstantBiasScale(None=ComponentMask(None=@n8) {b=true, g=true, r=true}) {bias=0.3, scale=0.66}) {luminanceFactors=<rgba 0.3 0.59 0.11 1>}
  n21 = ConstantBiasScale(None=Add(A=@n19, B=@n20)) {bias=-0.8, scale=8}
  n22 = Saturate(None=Subtract(A=Saturate(None=@n15), B=Clamp(None=@n21) {maxDefault=1.25}))
  n23 = Multiply(A=ScalarParameter() {defaultValue=0.1, parameterName=FoamValue}, B=@n22)
  n24 = Constant() {r=10}
  n25 = Constant() {r=10}
  n26 = ComponentMask(None=TextureCoordinate()) {g=true}
  n27 = Multiply(A=@n26, B=OneMinus(None=@n26))
  n28 = Saturate(None=Add(A=Constant() {r=-1.2}, B=Multiply(A=@n25, B=@n27)))
  n29 = Multiply(A=Multiply(A=@n28, B=@n28), B=Multiply(A=@n19, B=@n20))
  n30 = Saturate(None=SmoothStep(Value=Multiply(A=@n24, B=@n29)) {constMax=0.3})
  MaterialFunctionCall(Emissive=Multiply(A=@n1.RGB, B=Subtract(A=@n23, B=@n30))) {materialFunction=&MF_FX_EmissiveAdjustments}
  n31 = ScalarParameter() {defaultValue=0.25, parameterName=WhiteFoamAmt}
  n32 = Constant() {r=8}
  n33 = ComponentMask(None=TextureCoordinate()) {g=true}
  n34 = Multiply(A=@n33, B=OneMinus(None=@n33))
  n35 = Saturate(None=Add(A=Constant() {r=-1.2}, B=Multiply(A=@n32, B=@n34)))
  n36 = Multiply(A=Multiply(A=Multiply(A=@n35, B=@n35), B=Multiply(A=@n19, B=@n20))) {constB=5}
  n37 = Saturate(None=Subtract(A=Multiply(A=@n22, B=@n31), B=SmoothStep(Value=@n36) {constMax=0.5}))
  n38 = Constant() {r=5}
  Multiply(
   A=MaterialFunctionCall(
    Input=MaterialFunctionCall() {materialFunction=&LinearGradient}.UGradient
    Input Low=Constant()
    Input High=Constant() {r=0.3}
   ) {materialFunction=&RemapValue_0-1}.Result
   B=Multiply(A=@n1.A, B=Multiply(A=@n37, B=@n38))
  )
  Saturate(None=ConstantBiasScale(None=@n8) {bias=-0.25, scale=1})
 ]
}
