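# Translucent unlit Niagara ribbon material for the trail of the ToonyWater vehicle boost.
# Panning, distorted cloud textures build a mask that blends particle-color-tinted caustics cubemap shading with a lighter tone scaled by a dynamic parameter.
# Panning water normal maps warp ribbon UVs to carve the trail silhouette, which multiplies both color, fed through MF_FX_EmissiveAdjustments, and particle alpha.
# A near-duplicate chain lacks that function and the cubemap texture assignment.
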Material /VehicleCosmetics/Boosts/Boost_ToonyWater/Materials/M_Boost_ToonyWater_Trail_TW
graph {
 properties {
  bUsedWithNiagaraRibbons=true
  blendMode=BLEND_Translucent
  pixelDepthOffsetMode=PDOM_Legacy
  refractionMethod=RM_PixelNormalOffset
  shadingModel=MSM_Unlit
  usedShadingModels=Unlit
 }
 nodes [
  n1 = Constant() {r=0.5}
  n2 = TextureSample() {texture=/VehicleCosmetics/Boosts/Boost_ToonyWater/Texture/T_Boost_ToonyWater_Caustics_CubeMap}
  n3 = Saturate(None=ConstantBiasScale(None=Multiply(A=@n1, B=@n2.RGB)) {bias=0.5})
  n4 = Constant() {r=0.85}
  n5 = Multiply(A=@n3, B=@n4)
  n6 = Saturate(None=Multiply(A=Add(A=@n5, B=@n4), B=DynamicParameter().R))
  n7 = ParticleColor()
  n8 = Multiply(A=@n3, B=@n7.RGB)
  n9 = Constant() {r=0.5}
  n10 = TextureSample(UVs=Panner(Coordinate=TextureCoordinate() {uTiling=0.125, vTiling=0.125}) {speedX=-0.125, speedY=0.01}) {texture=/VehicleCosmetics/Boosts/Boost_ToonyWater/Texture/T_Boost_ToonyWater_Clouds_RGB_02_T}
  n11 = AppendVector(A=@n10.R, B=@n10.R)
  n12 = Multiply(A=Constant() {r=0.05}, B=@n11)
  n13 = Add(A=TextureCoordinate() {uTiling=0.75, vTiling=0.25}, B=@n12)
  n14 = Add(A=TextureCoordinate() {vTiling=0.25}, B=@n12)
  n15 = Add(
   A=TextureSample(UVs=Panner(Coordinate=@n13) {speedX=-1.25, speedY=-0.05}) {texture=&T_Boost_ToonyWater_Clouds_RGB_02_T}.RGB
   B=TextureSample(UVs=Panner(Coordinate=@n14) {speedX=-1.5, speedY=0.02}) {texture=&T_Boost_ToonyWater_Clouds_RGB_02_T}.RGB
  )
  n16 = ComponentMask(None=@n15) {g=true, r=true}
  n17 = ComponentMask(None=Add(A=TextureCoordinate(), B=Multiply(A=@n9, B=@n16))) {g=true}
  n18 = Saturate(None=ConstantBiasScale(None=@n17) {bias=-0.155, scale=1})
  n19 = Saturate(None=ConstantBiasScale(None=OneMinus(None=@n17)) {bias=-0.035, scale=1})
  n20 = ConstantBiasScale(None=Multiply(A=Constant() {r=6}, B=Multiply(A=@n18, B=@n19))) {bias=-0.1, scale=1.5}
  n21 = ComponentMask(None=TextureCoordinate()) {g=true}
  n22 = OneMinus(None=@n21)
  n23 = Power(Base=Multiply(A=Constant() {r=8}, B=Multiply(A=@n21, B=@n22)), Exp=Constant() {r=3})
  n24 = TextureCoordinate()
  n25 = ComponentMask(None=@n24) {r=true}
  n26 = Add(A=Multiply(A=Saturate(None=@n23), B=Saturate(None=@n20)), B=OneMinus(None=@n25))
  n27 = ComponentMask(None=@n15) {r=true}
  n28 = ComponentMask(None=@n15) {g=true}
  n29 = Clamp(None=ConstantBiasScale(None=Multiply(A=@n27, B=@n28)) {bias=-0.075, scale=8}) {maxDefault=1.25}
  n30 = Clamp(None=ConstantBiasScale(None=Add(A=@n27, B=@n28)) {bias=-0.35, scale=8}) {maxDefault=1.25}
  n31 = Constant() {r=3}
  n32 = Power(Base=@n25, Exp=Constant() {r=1.5})
  n33 = Multiply(A=Add(A=Add(A=@n29, B=@n30), B=Multiply(A=@n25, B=@n25)), B=Saturate(None=Multiply(A=@n31, B=@n32)))
  n34 = Saturate(None=Multiply(A=@n32, B=Constant() {r=10}))
  n35 = Saturate(None=Subtract(A=Saturate(None=@n26), B=Multiply(A=@n33, B=@n34)))
  n36 = Multiply(A=@n11, B=Constant() {r=0.15})
  n37 = Add(A=TextureCoordinate() {uTiling=3, vTiling=0.25}, B=@n36)
  n38 = Multiply(A=Time(), B=Constant() {r=3})
  n39 = Add(A=@n36, B=TextureCoordinate() {uTiling=4, vTiling=0.25})
  n40 = Add(
   A=TextureSample(UVs=Panner(Coordinate=@n37, Time=@n38) {speedX=-1, speedY=-0.05}) {texture=/VehicleCosmetics/Boosts/Boost_ToonyWater/Texture/T_Boost_ToonyWater_Water_02_N}.RGB
   B=TextureSample(UVs=Panner(Coordinate=@n39, Time=@n38) {speedX=-1, speedY=0.02}) {texture=&T_Boost_ToonyWater_Water_02_N}.RGB
  )
  n41 = Add(A=@n24, B=Multiply(A=Constant() {r=0.5}, B=ComponentMask(None=@n40) {g=true, r=true}))
  n42 = ComponentMask(None=@n41) {g=true}
  n43 = OneMinus(None=@n42)
  n44 = Multiply(A=Saturate(None=ConstantBiasScale(None=@n42) {bias=-0.285, scale=1}), B=Saturate(None=ConstantBiasScale(None=@n43) {bias=0.2, scale=1}))
  n45 = ConstantBiasScale(None=OneMinus(None=ComponentMask(None=@n41) {r=true})) {bias=0.2, scale=1}
  n46 = Multiply(A=Multiply(A=@n44, B=Saturate(None=@n45)), B=Constant() {r=5})
  n47 = Multiply(A=@n25, B=Constant() {r=18})
  n48 = Multiply(A=Saturate(None=ConstantBiasScale(None=@n46) {bias=-0.15, scale=4}), B=Saturate(None=Multiply(A=@n23, B=@n47)))
  MaterialFunctionCall(Emissive=Multiply(A=LinearInterpolate(A=@n6, B=@n8, Alpha=@n35), B=@n48)) {materialFunction=&MF_FX_EmissiveAdjustments}
  Multiply(A=@n7.A, B=@n48)
  n49 = Constant() {r=0.5}
  n50 = TextureSample()
  n51 = Saturate(None=ConstantBiasScale(None=Multiply(A=@n49, B=@n50.RGB)) {bias=0.5})
  n52 = Constant() {r=0.85}
  n53 = Multiply(A=Add(A=Multiply(A=@n51, B=@n52), B=@n52), B=DynamicParameter().R)
  n54 = ParticleColor()
  n55 = ComponentMask(None=TextureCoordinate()) {g=true}
  n56 = OneMinus(None=@n55)
  n57 = Power(Base=Multiply(A=Constant() {r=8}, B=Multiply(A=@n55, B=@n56)), Exp=Constant() {r=3})
  n58 = Constant() {r=0.5}
  n59 = TextureSample(UVs=Panner(Coordinate=TextureCoordinate() {uTiling=0.125, vTiling=0.125}) {speedX=-0.125, speedY=0.01}) {texture=&T_Boost_ToonyWater_Clouds_RGB_02_T}
  n60 = AppendVector(A=@n59.R, B=@n59.R)
  n61 = Multiply(A=Constant() {r=0.05}, B=@n60)
  n62 = Add(A=TextureCoordinate() {uTiling=0.75, vTiling=0.25}, B=@n61)
  n63 = Add(A=TextureCoordinate() {vTiling=0.25}, B=@n61)
  n64 = Add(
   A=TextureSample(UVs=Panner(Coordinate=@n62) {speedX=-1.25, speedY=-0.05}) {texture=&T_Boost_ToonyWater_Clouds_RGB_02_T}.RGB
   B=TextureSample(UVs=Panner(Coordinate=@n63) {speedX=-1.5, speedY=0.02}) {texture=&T_Boost_ToonyWater_Clouds_RGB_02_T}.RGB
  )
  n65 = ComponentMask(None=@n64) {g=true, r=true}
  n66 = ComponentMask(None=Add(A=TextureCoordinate(), B=Multiply(A=@n58, B=@n65))) {g=true}
  n67 = ConstantBiasScale(None=@n66) {bias=-0.155, scale=1}
  n68 = ConstantBiasScale(None=OneMinus(None=@n66)) {bias=-0.035, scale=1}
  n69 = Multiply(A=Constant() {r=6}, B=Multiply(A=Saturate(None=@n67), B=Saturate(None=@n68)))
  n70 = Multiply(A=Saturate(None=@n57), B=Saturate(None=ConstantBiasScale(None=@n69) {bias=-0.1, scale=1.5}))
  n71 = TextureCoordinate()
  n72 = ComponentMask(None=@n71) {r=true}
  n73 = OneMinus(None=@n72)
  n74 = ComponentMask(None=@n64) {r=true}
  n75 = ComponentMask(None=@n64) {g=true}
  n76 = ConstantBiasScale(None=Multiply(A=@n74, B=@n75)) {bias=-0.075, scale=8}
  n77 = ConstantBiasScale(None=Add(A=@n74, B=@n75)) {bias=-0.35, scale=8}
  n78 = Add(A=Add(A=Clamp(None=@n76) {maxDefault=1.25}, B=Clamp(None=@n77) {maxDefault=1.25}), B=Multiply(A=@n72, B=@n72))
  n79 = Power(Base=@n72, Exp=Constant() {r=1.5})
  n80 = Saturate(None=Multiply(A=Constant() {r=3}, B=@n79))
  n81 = Multiply(A=@n79, B=Constant() {r=10})
  n82 = Subtract(A=Saturate(None=Add(A=@n70, B=@n73)), B=Multiply(A=Multiply(A=@n78, B=@n80), B=Saturate(None=@n81)))
  n83 = Multiply(A=Constant() {r=0.15}, B=@n60)
  n84 = Add(A=TextureCoordinate() {uTiling=3, vTiling=0.25}, B=@n83)
  n85 = Multiply(A=Time(), B=Constant() {r=3})
  n86 = Add(A=@n83, B=TextureCoordinate() {uTiling=4, vTiling=0.25})
  n87 = Add(
   A=TextureSample(UVs=Panner(Coordinate=@n84, Time=@n85) {speedX=-1, speedY=-0.05}) {texture=&T_Boost_ToonyWater_Water_02_N}.RGB
   B=TextureSample(UVs=Panner(Coordinate=@n86, Time=@n85) {speedX=-1, speedY=0.02}) {texture=&T_Boost_ToonyWater_Water_02_N}.RGB
  )
  n88 = Add(A=@n71, B=Multiply(A=Constant() {r=0.5}, B=ComponentMask(None=@n87) {g=true, r=true}))
  n89 = ComponentMask(None=@n88) {g=true}
  n90 = OneMinus(None=@n89)
  n91 = Multiply(A=Saturate(None=ConstantBiasScale(None=@n89) {bias=-0.285, scale=1}), B=Saturate(None=ConstantBiasScale(None=@n90) {bias=0.2, scale=1}))
  n92 = ConstantBiasScale(None=OneMinus(None=ComponentMask(None=@n88) {r=true})) {bias=0.2, scale=1}
  n93 = Multiply(A=Multiply(A=@n91, B=Saturate(None=@n92)), B=Constant() {r=5})
  n94 = Multiply(A=@n72, B=Constant() {r=18})
  n95 = Multiply(A=Saturate(None=ConstantBiasScale(None=@n93) {bias=-0.15, scale=4}), B=Saturate(None=Multiply(A=@n57, B=@n94)))
  Multiply(A=LinearInterpolate(A=Saturate(None=@n53), B=Multiply(A=@n51, B=@n54.RGB), Alpha=Saturate(None=@n82)), B=@n95)
  Multiply(A=@n54.A, B=@n95)
  Multiply(A=Saturate(None=ConstantBiasScale(None=@n91) {bias=-0.1, scale=5})) {constB=0}
  ReflectionVectorWS()
  ReflectionVectorWS()
  Saturate(None=ConstantBiasScale(None=@n87) {bias=0.5, scale=0.8})
  Saturate(None=ConstantBiasScale(None=@n40) {bias=0.5, scale=0.8})
  Saturate(None=ConstantBiasScale(None=@n44) {bias=-0.1, scale=5})
 ]
}
