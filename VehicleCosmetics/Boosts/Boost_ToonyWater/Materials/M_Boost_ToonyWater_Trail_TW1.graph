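# Translucent unlit material for Niagara ribbons that draws the toony water boost trail.
# Cloud noise warps two panning water normal maps, whose sum distorts UV gradients into a wavy, tapered silhouette mask.
# Scrolling cloud layers build a streak mask blending two particle-color tints, both shaded by a caustics texture sampled with the tangent-space reflection vector.
# The silhouette masks the color fed to emissive adjustments and, with particle alpha and a U fade, forms an alpha mask.
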Material /VehicleCosmetics/Boosts/Boost_ToonyWater/Materials/M_Boost_ToonyWater_Trail_TW1
graph {
 properties {
  bUsedWithNiagaraRibbons=true
  blendMode=BLEND_Translucent
  pixelDepthOffsetMode=PDOM_Legacy
  refractionMethod=RM_PixelNormalOffset
  shadingModel=MSM_Unlit
  usedShadingModels=Unlit
 }
 nodes [
  n1 = TextureSample(UVs=Panner(Coordinate=TextureCoordinate() {uTiling=0.125, vTiling=0.125}) {speedX=-0.125, speedY=0.01}) {texture=/VehicleCosmetics/Boosts/Boost_ToonyWater/Textures/T_VCVFX_Boost_ToonyWater_Clouds2_RGB}
  n2 = AppendVector(A=@n1.R, B=@n1.R)
  n3 = Multiply(A=Constant() {r=0.15}, B=@n2)
  n4 = Add(A=TextureCoordinate() {uTiling=3, vTiling=0.25}, B=@n3)
  n5 = Multiply(A=Time(), B=Constant() {r=3})
  n6 = Add(A=@n3, B=TextureCoordinate() {uTiling=4, vTiling=0.25})
  n7 = Add(
   A=TextureSample(UVs=Panner(Coordinate=@n4, Time=@n5) {speedX=-1, speedY=-0.05}) {samplerType=SAMPLERTYPE_Normal, texture=/DealershipSharedLibrary/FX/Textures/Noise/T_VCVFX_Water02_N}.RGB
   B=TextureSample(UVs=Panner(Coordinate=@n6, Time=@n5) {speedX=-1, speedY=0.02}) {samplerType=SAMPLERTYPE_Normal, texture=&T_VCVFX_Water02_N}.RGB
  )
  MaterialExpressionConstantClamp_2 = Saturate(None=ConstantBiasScale(None=@n7) {bias=0.5, scale=0.8})
  MaterialExpressionVertexColor_0 = ParticleColor()
  n8 = TextureCoordinate()
  n9 = Add(A=Constant2Vector() {g=0.2}, B=@n8)
  n10 = Multiply(A=Constant() {r=0.5}, B=ComponentMask(None=@n7) {g=true, r=true})
  n11 = ComponentMask(None=Add(A=@n9, B=@n10)) {g=true}
  MaterialExpressionConstantClamp_5 = Saturate(None=Multiply(A=Subtract(A=@n11) {constB=0.2}) {constB=2})
  n12 = Add(A=Constant2Vector() {g=-0.2}, B=@n8)
  n13 = ComponentMask(None=Add(A=@n10, B=OneMinus(None=@n12))) {g=true}
  MaterialExpressionConstantClamp_6 = Saturate(None=Multiply(A=Subtract(A=@n13) {constB=0.2}) {constB=2})
  n14 = TextureCoordinate()
  n15 = OneMinus(None=Power(Base=ComponentMask(None=@n14) {r=true}) {constExponent=0.8})
  n16 = Multiply(
   A=Multiply(A=Multiply(A=@MaterialExpressionConstantClamp_5, B=@MaterialExpressionConstantClamp_6), B=Saturate(None=@n15))
   B=Constant() {r=5}
  )
  n17 = ComponentMask(None=TextureCoordinate()) {g=true}
  n18 = Multiply(A=Constant() {r=8}, B=Multiply(A=@n17, B=OneMinus(None=@n17)))
  n19 = ComponentMask(None=@n9) {r=true}
  n20 = Multiply(A=@n19, B=Constant() {r=18})
  n21 = Multiply(A=Saturate(None=ConstantBiasScale(None=@n16) {bias=-0.15, scale=4}), B=Saturate(None=Multiply(A=@n18, B=@n20)))
  n22 = MaterialFunctionCall() {materialFunction=&LinearGradient}
  n23 = Constant()
  n24 = Constant() {r=0.2}
  MaterialExpressionConstantClamp_3 = Saturate(None=Multiply(
    A=Multiply(A=@MaterialExpressionVertexColor_0.A, B=@n21)
    B=MaterialFunctionCall(Input=@n22.UGradient, Input Low=@n23, Input High=@n24) {materialFunction=&RemapValue_0-1}.Result
   ))
  n25 = Constant() {r=0.5}
  n26 = ReflectionVectorWS()
  n27 = TextureSample(UVs=ComponentMask(None=Transform(None=@n26) {transformSourceType=TRANSFORMSOURCE_World, transformType=TRANSFORM_Tangent}) {g=true, r=true}) {texture=/VehicleCosmetics/Boosts/Boost_ToonyWater/Textures/T_VCVFX_Boost_ToonyWater_Caustics_CubeMap}
  MaterialExpressionConstantClamp_9 = Saturate(None=ConstantBiasScale(None=Multiply(A=@n25, B=@n27.RGB)) {bias=0.5})
  n28 = Constant() {r=0.85}
  n29 = Multiply(A=@MaterialExpressionConstantClamp_9, B=@n28)
  MaterialExpressionConstantClamp_7 = Saturate(None=Multiply(A=Add(A=@n29, B=@n28), B=ScalarParameter() {defaultValue=1, parameterName=Param1}))
  n30 = Add(A=Multiply(A=@MaterialExpressionConstantClamp_7, B=@MaterialExpressionVertexColor_0.RGB), B=ScalarParameter() {parameterName=LightDeatails})
  n31 = Multiply(A=@MaterialExpressionConstantClamp_9, B=@MaterialExpressionVertexColor_0.RGB)
  MaterialExpressionConstantClamp_11 = Saturate(None=Power(Base=@n18, Exp=Constant() {r=3}))
  n32 = Constant() {r=6}
  n33 = Constant() {r=0.5}
  n34 = Multiply(A=Constant() {r=0.05}, B=@n2)
  n35 = Add(A=TextureCoordinate() {uTiling=0.75, vTiling=0.25}, B=@n34)
  n36 = Add(A=TextureCoordinate() {vTiling=0.25}, B=@n34)
  n37 = Add(
   A=TextureSample(UVs=Panner(Coordinate=@n35) {speedX=-1.25, speedY=-0.05}) {texture=&T_VCVFX_Boost_ToonyWater_Clouds2_RGB}.RGB
   B=TextureSample(UVs=Panner(Coordinate=@n36) {speedX=-1.5, speedY=0.02}) {texture=&T_VCVFX_Boost_ToonyWater_Clouds2_RGB}.RGB
  )
  n38 = ComponentMask(None=@n37) {g=true, r=true}
  n39 = ComponentMask(None=Add(A=TextureCoordinate(), B=Multiply(A=@n33, B=@n38))) {g=true}
  MaterialExpressionConstantClamp_13 = Saturate(None=ConstantBiasScale(None=@n39) {bias=-0.155, scale=1})
  MaterialExpressionConstantClamp_14 = Saturate(None=ConstantBiasScale(None=OneMinus(None=@n39)) {bias=-0.035, scale=1})
  n40 = Multiply(A=@MaterialExpressionConstantClamp_13, B=@MaterialExpressionConstantClamp_14)
  MaterialExpressionConstantClamp_12 = Saturate(None=ConstantBiasScale(None=Multiply(A=@n32, B=@n40)) {bias=-0.1, scale=1.5})
  MaterialExpressionConstantClamp_10 = Saturate(None=Add(A=Multiply(A=@MaterialExpressionConstantClamp_11, B=@MaterialExpressionConstantClamp_12), B=OneMinus(None=@n19)))
  n41 = ComponentMask(None=@n37) {r=true}
  n42 = ComponentMask(None=@n37) {g=true}
  MaterialExpressionConstantClamp_17 = Saturate(None=ConstantBiasScale(None=Multiply(A=@n41, B=@n42)) {bias=-0.075, scale=8})
  MaterialExpressionConstantClamp_18 = Saturate(None=ConstantBiasScale(None=Add(A=@n41, B=@n42)) {bias=-0.35, scale=8})
  n43 = Power(Base=@n19, Exp=Constant() {r=1.5})
  MaterialExpressionConstantClamp_16 = Saturate(None=Multiply(A=@n43, B=Constant() {r=3}))
  n44 = Multiply(
   A=Add(A=Add(A=@MaterialExpressionConstantClamp_17, B=@MaterialExpressionConstantClamp_18), B=Multiply(A=@n19, B=@n19))
   B=@MaterialExpressionConstantClamp_16
  )
  MaterialExpressionConstantClamp_15 = Saturate(None=Multiply(A=@n43, B=Constant() {r=10}))
  MaterialExpressionConstantClamp_8 = Saturate(None=Subtract(A=@MaterialExpressionConstantClamp_10, B=Multiply(A=@n44, B=@MaterialExpressionConstantClamp_15)))
  MaterialFunctionCall(Emissive=Multiply(A=LinearInterpolate(A=@n30, B=@n31, Alpha=@MaterialExpressionConstantClamp_8), B=@n21)) {materialFunction=&MF_FX_EmissiveAdjustments}
  TextureSample() {texture=&T_VCVFX_Boost_ToonyWater_Clouds2_RGB}
 ]
}
